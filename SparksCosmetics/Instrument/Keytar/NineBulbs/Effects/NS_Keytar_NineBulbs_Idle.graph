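# Niagara idle effect for the Nine Bulbs keytar cosmetic, placing glowing, flickering bulb lights at fixed offsets on the instrument.
# StarLights bursts array-positioned particles, culled by X position, with sprite, light and mesh renderers; HighLights adds two larger lights with sprites.
# Smoothed random values drive flicker, remapped by camera distance, with brightness or alpha also following time of day.
# The stepped flash pattern's ScaleColor and the rainbow color modules are disabled.
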NiagaraSystem /SparksCosmetics/Instrument/Keytar/NineBulbs/Effects/NS_Keytar_NineBulbs_Idle
dependencies {
 usedDataInterfaces [
  &NiagaraDataInterfaceArrayFloat3
  &NiagaraDataInterfaceColorCurve
  &NiagaraDataInterfaceCurve
  &NiagaraDataInterfaceArrayFloat
 ]
 usedDynamicInputs []
 usedModules [&SystemState]
}
emitters {
 HighLights {
  emitterUpdateScript [
   &EmitterState()
   &SpawnBurst_Instantaneous(Spawn Count=2, Use Spawn Probability=false, UseLoopCountLimit=true)
  ]
  particleSpawnScript [
   &PartitionParticles()
   &InitializeParticle(
    Color=<rgba 5 4.6008 2.215 1>
    Position Offset=&SelectVectorFromArray(
     Direct Array Index=$Particles.PartitionParticles.Partition
     Vector Selection Array=DataInterface() {propertyValues {FloatData [<3.5 12 13>, <2.4249 12 50>, <-11 6 18>]}}
    )
    Sprite Size Mode=::NewEnumerator3
    UsePositionOffset=true
    Write Material Random=true
   )
  ]
  particleUpdateScript [
   &ParticleState(Kill Particles When Lifetime Has Elapsed=false, Let Infinitely Lived Particles Die When Emitter Deactivates=true)
   &InterpolateOverTime(
    Initial Float Value=1
    Rate Of Change=2
    Target Float Value:NiagaraFloat=&RandomRangeFloat(Minimum=0, Maximum=2, Evaluation Type=::NewEnumerator1)
    Use Set Interpolation Rate=true
   )
   &ScaleColor(
    Scale RGB=&VectorFromFloat(Value=&FloatFromCurve(
      FloatCurve=DataInterface() {propertyValues {Curve {keys [
          {interpMode=RCIM_Cubic, tangentMode=RCTM_User, time=0, value=1}
          {interpMode=RCIM_Cubic, tangentMode=RCTM_User, time=0.5, value=0.75}
          {interpMode=RCIM_Cubic, tangentMode=RCTM_User, time=1, value=1}
         ]}}}
      CurveIndex=Linked() {linkedVariable {name=NPC.FortniteNPC.TimeOfDay_0-1, type=NiagaraFloat}}
      Scale Curve=&FN_LOD_Distance_Remap(
       Start Distance=300
       End Distance=1000
       Minimum Output=$Particles.InterpolateOverTime.InterpolateOverTime_Float.Moving Average
       Maximum Output=&Add_Float(A=0.5, B=&Multiply_Float(A=$Particles.InterpolateOverTime.InterpolateOverTime_Float.Moving Average, B=0.25))
       Clamp Output=true
      )
     ))
    ScaleA=false
    ScaleRGB=true
   )
   &FN_Light_Attributes(
    Color=&MultiplyLinearColor(Linear Color=$Particles.Color, Scale Factor=<2 1 0 1>)
    Create Color Attribute=true
    Create Exponent Attribute=true
    Create Radius Attribute=true
    Exponent=2
    Radius=&SelectFloatFromArray(Direct Array Index=$Particles.PartitionParticles.Partition, Float Selection Array=DataInterface() {propertyValues {FloatData [20,30,15]}})
   )
  ]
  rendererClasses [&NiagaraLightRendererProperties, &NiagaraSpriteRendererProperties]
  renderers [
   {rendererClass=&NiagaraLightRendererProperties, rendererIndex=0}
   {rendererClass=&NiagaraSpriteRendererProperties, rendererIndex=1}
  ]
 }
 StarLights {
  emitterUpdateScript [
   &EmitterState()
   &SpawnBurst_Instantaneous(Spawn Count=12, Use Spawn Probability=false, UseLoopCountLimit=true)
  ]
  particleSpawnScript [
   &PartitionParticles(Number Of Partitions=12)
   &InitializeParticle(
    Color=<rgba 5 4.6008 2.215 0.1>
    Position Offset=&SelectVectorFromArray(
     Direct Array Index=$Particles.PartitionParticles.Partition
     Vector Selection Array=DataInterface() {propertyValues {FloatData [<-5 5 56.75>, <0 0 0>, <11 5 54.5>, <0 0 0>, <0 0 0>, <11 5 30.25>, <0 0 0>, <10.5 5 12.5>, <13 5 5.5>, <0 0 0>, <1.75 5 1.75>, <0 0 0>]}}
    )
    Sprite Size Mode=::NewEnumerator3
    Uniform Sprite Size=7
    UsePositionOffset=true
    Write Material Random=true
   )
   &CameraOffset(Camera Offset Amount=5)
   &Color(Color=&ColorFromCurve(
     ColorCurve=DataInterface() {propertyValues {
       AlphaCurve {keys [<at 0 1 RCIM_Cubic>]}
       BlueCurve {keys [<at 0 0 RCIM_Cubic>,<at 0.2 0>,<at 0.4 1 RCIM_Cubic>,<at 0.6 0>,<at 0.8 0>,<at 1 1>]}
       GreenCurve {keys [
         <at 0 0 RCIM_Cubic>
         <at 0.2 1>
         {arriveTangent=-2.5, interpMode=RCIM_Cubic, leaveTangent=-2.5, time=0.4, value=0}
         <at 0.6 0>
         <at 0.8 1>
         <at 1 0>
        ]}
       RedCurve {keys [
         <at 0 1 RCIM_Cubic>
         <at 0.2 0>
         {arriveTangent=2.5, interpMode=RCIM_Cubic, leaveTangent=2.5, time=0.4, value=0}
         <at 0.6 1>
         <at 0.8 0>
         <at 1 0>
        ]}
      }}
     CurveIndex=&Multiply_FloatByInt(Float=0.0556, Integer=$Particles.PartitionParticles.Partition)
    )) {enabled=false}
   &KillParticles(Kill Particles=&SetBoolByFloatComparison(
     A=&MakeFloatFromVector(Vector=Linked() {linkedVariable {name=Particles.Position, type=NiagaraPosition}}, Channel=::NewEnumerator0)
     B=0
     Comparison Type=::NewEnumerator4
    ))
  ]
  particleUpdateScript [
   &ParticleState(Kill Particles When Lifetime Has Elapsed=false, Let Infinitely Lived Particles Die When Emitter Deactivates=true)
   &SolveForcesAndVelocity(Clamp Velocity=false, Limit Acceleration=false)
   /SetVariables(Transient.FlashSpeed:NiagaraFloat=&Multiply_Float(A=$Engine.Time, B=3), Transient.FlashSteps=6) {isSetParameters=true}
   /SetVariables(Transient.FlashPattern:NiagaraFloat=&MakeCustomFloatFromBool(
     Bool=&SetBoolByIntComparison(
      A=&ModuloInt(A=$Particles.PartitionParticles.Partition, B=$Transient.FlashSteps)
      B=&ModuloInt(A=&MakeIntFromFloat(Float=$Transient.FlashSpeed), B=$Transient.FlashSteps)
      Comparison Type=::NewEnumerator4
     )
     True Float=1
     False Float=0
    )) {isSetParameters=true}
   &InterpolateOverTime(
    Initial Float Value=1
    Rate Of Change=2
    Target Float Value:NiagaraFloat=&RandomRangeFloat(Minimum=0, Maximum=2, Evaluation Type=::NewEnumerator1)
    Use Set Interpolation Rate=true
   )
   &ScaleColor(
    Scale Alpha=&FloatFromCurve(
     FloatCurve=DataInterface() {propertyValues {Curve {keys [
         {interpMode=RCIM_Cubic, tangentMode=RCTM_User, time=0, value=1}
         {interpMode=RCIM_Cubic, tangentMode=RCTM_User, time=0.5, value=0.75}
         {interpMode=RCIM_Cubic, tangentMode=RCTM_User, time=1, value=1}
        ]}}}
     CurveIndex=Linked() {linkedVariable {name=NPC.FortniteNPC.TimeOfDay_0-1, type=NiagaraFloat}}
     Scale Curve=&FN_LOD_Distance_Remap(
      Start Distance=300
      End Distance=1000
      Minimum Output=$Particles.InterpolateOverTime.InterpolateOverTime_Float.Moving Average
      Maximum Output=&Add_Float(A=0.5, B=&Multiply_Float(A=$Particles.InterpolateOverTime.InterpolateOverTime_Float.Moving Average, B=0.25))
      Clamp Output=true
     )
    )
    ScaleA=true
    ScaleRGB=false
   )
   &ScaleColor(
    Scale Alpha=&FloatFromCurve(
     FloatCurve=DataInterface() {propertyValues {Curve {keys [
         {interpMode=RCIM_Cubic, tangentMode=RCTM_User, time=0, value=1}
         {interpMode=RCIM_Cubic, tangentMode=RCTM_User, time=0.5, value=0.75}
         {interpMode=RCIM_Cubic, tangentMode=RCTM_User, time=1, value=1}
        ]}}}
     CurveIndex=Linked() {linkedVariable {name=NPC.FortniteNPC.TimeOfDay_0-1, type=NiagaraFloat}}
     Scale Curve=$Transient.FlashPattern
    )
    Scale RGB=&VectorFromFloat(Value=&FloatFromCurve(
      FloatCurve=DataInterface() {propertyValues {Curve {keys [
          {interpMode=RCIM_Cubic, tangentMode=RCTM_User, time=0, value=1}
          {interpMode=RCIM_Cubic, tangentMode=RCTM_User, time=0.5, value=0.75}
          {interpMode=RCIM_Cubic, tangentMode=RCTM_User, time=1, value=1}
         ]}}}
      CurveIndex=Linked() {linkedVariable {name=NPC.FortniteNPC.TimeOfDay_0-1, type=NiagaraFloat}}
      Scale Curve=$Transient.FlashPattern
     ))
    ScaleA=false
    ScaleRGB=true
   ) {enabled=false, module=ScaleColor001}
   &ScaleSpriteSize(
    Scale Sprite Size Mode=::NewEnumerator0
    Uniform Scale Factor=&Add_Float(
     A=&FN_LOD_Distance_Remap(Start Distance=300, End Distance=1000, Minimum Output=0.8, Maximum Output=-0.5, Clamp Output=true)
     B=&Multiply_Float(
      A=$Particles.InterpolateOverTime.InterpolateOverTime_Float.Moving Average
      B=&FN_LOD_Distance_Remap(Start Distance=300, End Distance=1000, Minimum Output=0.2, Maximum Output=2, Clamp Output=true)
     )
    )
   )
   &Color(Color=&ColorFromCurve(
     ColorCurve=DataInterface() {propertyValues {
       AlphaCurve {keys [<at 0 1 RCIM_Cubic>]}
       BlueCurve {keys [<at 0 1>, <at 0.01 0 RCIM_Cubic>, <at 0.5 0>, {arriveTangent=2, interpMode=RCIM_Cubic, leaveTangent=2, time=0.99, value=1}, <at 1 1>]}
       GreenCurve {keys [<at 0 1>,<at 0.01 0 RCIM_Cubic>,<at 0.5 1>,<at 0.99 0 RCIM_Cubic>,<at 1 1>]}
       RedCurve {keys [<at 0 1>, <at 0.01 1 RCIM_Cubic>, <at 0.5 0>, {arriveTangent=2, interpMode=RCIM_Cubic, leaveTangent=2, time=0.99, value=0}, <at 1 1>]}
      }}
     CurveIndex=&FN_LOD_Distance_Remap(Start Distance=300, End Distance=1000, Minimum Output=0, Maximum Output=1, Clamp Output=true)
    )) {enabled=false, module=Color001}
   &Color(Color=&ColorFromCurve(
     ColorCurve=DataInterface() {propertyValues {
       AlphaCurve {keys [<at 0 1 RCIM_Cubic>]}
       BlueCurve {keys [<at 0 0 RCIM_Cubic>, <at 0.5 0>, {interpMode=RCIM_Cubic, leaveTangent=2, time=1, value=1}]}
       GreenCurve {keys [<at 0 0 RCIM_Cubic>,<at 0.5 1>,<at 1 0 RCIM_Cubic>]}
       RedCurve {keys [<at 0 1 RCIM_Cubic>, <at 0.5 0>, {interpMode=RCIM_Cubic, leaveTangent=2, time=1, value=0}]}
      }}
     CurveIndex=&Multiply_FloatByInt(Float=0.0833, Integer=$Particles.PartitionParticles.Partition)
    )) {enabled=false, module=Color002}
   &FN_Light_Attributes(
    Color=&MultiplyLinearColor(Linear Color=$Particles.Color, Scale Factor=<2 1 0 1>)
    Create Color Attribute=true
    Create Exponent Attribute=true
    Create Radius Attribute=true
    Exponent=3
    Radius=10
   )
  ]
  rendererClasses [
   &NiagaraSpriteRendererProperties
   &NiagaraLightRendererProperties
   &NiagaraMeshRendererProperties
  ]
  renderers [
   {rendererClass=&NiagaraSpriteRendererProperties}
   {rendererClass=&NiagaraLightRendererProperties, rendererIndex=1}
   {rendererClass=&NiagaraMeshRendererProperties, rendererIndex=2}
  ]
 }
}
summary {
 systemName=NS_Keytar_NineBulbs_Idle
 emitters [
  {
   emitterName=StarLights
   rendererClasses [
    &NiagaraSpriteRendererProperties
    &NiagaraLightRendererProperties
    &NiagaraMeshRendererProperties
   ]
  }
  {emitterName=HighLights, rendererClasses [&NiagaraLightRendererProperties, &NiagaraSpriteRendererProperties]}
 ]
 userVariables []
}
