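# Physics asset for the NineBulbs keytar instrument cosmetic, giving simple bodies and joints to its hinge and dangling spider.
# A box-shaped handle body anchors a swinging capsule hinge and a three-link spiderweb chain that carries the spider root.
# Short capsule chains for the center, back, front and top-middle legs branch from that root with mostly tight swing limits.
# Twist is locked on every joint except the fully free last left top-middle leg joint.
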PhysicsAsset /SparksCosmetics/Instrument/Keytar/NineBulbs/Meshes/NineBulbs_Keytar_Physics
bodies {
 dyn_center_leg_01 [{shapeType=Capsule, length=1.5177, radius=0.3, shapeName=dyn_center_leg_01_capsule, center=<0.7974 -0.0419 -0.0058>, rotation=<rot 0 90 -90>}]
 dyn_center_leg_l_02 [{
   shapeType=Capsule
   length=1.6639
   radius=0.3
   shapeName=dyn_center_leg_l_02_capsule
   center=<0.9887 -0.0237 -0.0573>
   rotation=<rot 0 90 -90>
  }]
 dyn_center_leg_l_03 [{shapeType=Capsule, length=0.9386, radius=0.3, shapeName=dyn_center_leg_l_03_capsule, center=<0.512 -0.0123 -0.0082>, rotation=<rot 0 90 -90>}]
 dyn_center_leg_r_01 [{shapeType=Capsule, length=1.7814, radius=0.3, shapeName=dyn_center_leg_r_01_capsule, center=<1.1156 0.0391 0.0657>, rotation=<rot 0 90 -90>}]
 dyn_center_leg_r_02 [{shapeType=Capsule, length=1.7906, radius=0.3, shapeName=dyn_center_leg_r_02_capsule, center=<0.8734 0.0294 0.0391>, rotation=<rot 0 90 -90>}]
 dyn_center_leg_r_03 [{
   shapeType=Capsule
   length=0.9386
   radius=0.3
   shapeName=dyn_center_leg_r_03_capsule
   center=<0.5121 -0.0031 -0.0051>
   rotation=<rot 0 90 -90>
  }]
 dyn_hinge [
  {shapeType=Capsule, length=4, radius=2, shapeName=dyn_hinge_capsule, rotation=<rot 0 -90 0>}
  {
   shapeType=Capsule
   length=6.0724
   radius=4.0643
   shapeName=dyn_hinge_capsule
   center=<6.2958 0.2467 6.931>
   rotation=<rot -0.0047 -90 124.8627>
  }
 ]
 dyn_spider_back_l_01 [{shapeType=Capsule, length=1, radius=0.25, shapeName=dyn_spider_back_l_01_capsule, center=<0.5478 0 0>, rotation=<rot 0 -90 90>}]
 dyn_spider_back_l_02 [{shapeType=Capsule, length=1.5, radius=0.25, shapeName=dyn_spider_back_l_02_capsule, center=<0.9805 0 0.0801>, rotation=<rot 0 -90 90>}]
 dyn_spider_back_r_01 [{shapeType=Capsule, length=1, radius=0.25, shapeName=dyn_spider_back_r_01_capsule, center=<0.518 0 0>, rotation=<rot 0 -90 90>}]
 dyn_spider_back_r_02 [{shapeType=Capsule, length=1.5283, radius=0.25, shapeName=dyn_spider_back_r_02_capsule, center=<1.0588 0 0>, rotation=<rot 0 -90 90>}]
 dyn_spider_front_l_01 [{shapeType=Capsule, length=0.5, radius=0.25, shapeName=dyn_spider_front_l_01_capsule, center=<0.2729 0 0>, rotation=<rot 0 -90 90>}]
 dyn_spider_front_l_02 [{shapeType=Capsule, length=0.75, radius=0.25, shapeName=dyn_spider_front_l_02_capsule, center=<0.3583 0 0>, rotation=<rot 0 -90 90>}]
 dyn_spider_front_r_01 [{shapeType=Capsule, length=0.5, radius=0.25, shapeName=dyn_spider_front_r_01_capsule, center=<0.4057 0 0>, rotation=<rot 0 -90 90>}]
 dyn_spider_front_r_02 [{shapeType=Capsule, length=0.75, radius=0.25, shapeName=dyn_spider_front_r_02_capsule, center=<0.4133 0 0>, rotation=<rot 0 -90 90>}]
 dyn_spider_root [{
   shapeType=Capsule
   length=1.4182
   radius=0.8511
   shapeName=dyn_spider_root_capsule
   center=<1.5689 -0.0614 -0.0199>
   rotation=<rot 0 -95.9907 -89.2881>
  }]
 dyn_spider_top_middle_l_01 [{shapeType=Capsule, length=1, radius=0.25, shapeName=dyn_spider_top_middle_l_01_capsule, center=<0.7243 0 0>, rotation=<rot 0 -90 90>}]
 dyn_spider_top_middle_l_02 [{shapeType=Capsule, length=1.5, radius=0.25, shapeName=dyn_spider_top_middle_l_02_capsule, center=<0.7376 0 0>, rotation=<rot 0 -90 90>}]
 dyn_spider_top_middle_l_03 [{shapeType=Capsule, length=1.4375, radius=0.25, shapeName=dyn_spider_top_middle_l_03_capsule, center=<1.1543 0 0>, rotation=<rot 0 -90 90>}]
 dyn_spider_top_middle_r_01 [{shapeType=Capsule, length=1.3, radius=0.25, shapeName=dyn_spider_top_middle_r_01_capsule, center=<0.7248 0 0>, rotation=<rot 0 -90 90>}]
 dyn_spider_top_middle_r_02 [{shapeType=Capsule, length=1.5, radius=0.25, shapeName=dyn_spider_top_middle_r_02_capsule, center=<0.8057 0 0>, rotation=<rot 0 -90 90>}]
 dyn_spider_top_middle_r_03 [{shapeType=Capsule, length=1.5, radius=0.25, shapeName=dyn_spider_top_middle_r_03_capsule, center=<0.8215 0 0>, rotation=<rot 0 -90 90>}]
 dyn_spiderweb_01 [{shapeType=Capsule, length=1.8103, radius=1, shapeName=dyn_spiderweb_01_capsule, center=<1.5 0.75 0>, rotation=<rot -0.0001 -71.4152 -90>}]
 dyn_spiderweb_02 [{shapeType=Capsule, length=2.0607, radius=1.3324, shapeName=dyn_spiderweb_02_capsule, center=<1.5 0.7432 0>, rotation=<rot 0 -90 -90>}]
 dyn_spiderweb_03 [{shapeType=Capsule, length=1.6221, radius=1.3781, shapeName=dyn_spiderweb_03_capsule, center=<1.5 0.664 0>, rotation=<rot 0 -90 -90>}]
 handle [
  {
   shapeType=Box
   extentX=20.0165
   extentY=6.7634
   extentZ=32.9243
   shapeName=handle_box
   center=<-0.2733 -7.6651 21.2069>
   rotation=<rot 0 -90 0>
  }
  {
   shapeType=Box
   extentX=4.4854
   extentY=6.6677
   extentZ=23.3417
   shapeName=handle_box
   center=<-0.2593 -18.4035 22.4484>
   rotation=<rot 0 -90 0>
  }
 ]
}
constraints [
 {bone1Name=dyn_spiderweb_01, bone2Name=handle, swing1LimitDegrees=10, swing2LimitDegrees=120, twistLimitDegrees=45, twistMotion=Locked}
 {
  bone1Name=dyn_spiderweb_02
  bone2Name=dyn_spiderweb_01
  swing1LimitDegrees=20
  swing2LimitDegrees=75
  twistLimitDegrees=45
  twistMotion=Locked
 }
 {
  bone1Name=dyn_spiderweb_03
  bone2Name=dyn_spiderweb_02
  swing1LimitDegrees=20
  swing2LimitDegrees=60
  twistLimitDegrees=45
  twistMotion=Locked
 }
 {
  bone1Name=dyn_hinge
  bone2Name=handle
  swing1LimitDegrees=12
  swing2LimitDegrees=45
  swing2Motion=Locked
  twistLimitDegrees=10
  twistMotion=Locked
 }
 {
  bone1Name=dyn_center_leg_01
  bone2Name=dyn_spider_root
  swing1LimitDegrees=5
  swing1Motion=Locked
  swing2LimitDegrees=10
  swing2Motion=Locked
  twistLimitDegrees=1
  twistMotion=Locked
 }
 {
  bone1Name=dyn_center_leg_l_02
  bone2Name=dyn_center_leg_01
  swing1LimitDegrees=1
  swing1Motion=Locked
  swing2LimitDegrees=20
  twistLimitDegrees=1
  twistMotion=Locked
 }
 {
  bone1Name=dyn_center_leg_l_03
  bone2Name=dyn_center_leg_l_02
  swing1LimitDegrees=1
  swing1Motion=Locked
  swing2LimitDegrees=20
  twistLimitDegrees=1
  twistMotion=Locked
 }
 {
  bone1Name=dyn_center_leg_r_01
  bone2Name=dyn_spider_root
  swing1LimitDegrees=0
  swing1Motion=Locked
  swing2LimitDegrees=4
  swing2Motion=Locked
  twistLimitDegrees=1
  twistMotion=Locked
 }
 {
  bone1Name=dyn_center_leg_r_02
  bone2Name=dyn_center_leg_r_01
  swing1LimitDegrees=1
  swing1Motion=Locked
  swing2LimitDegrees=20
  twistLimitDegrees=1
  twistMotion=Locked
 }
 {
  bone1Name=dyn_center_leg_r_03
  bone2Name=dyn_center_leg_r_02
  swing1LimitDegrees=1
  swing1Motion=Locked
  swing2LimitDegrees=20
  twistLimitDegrees=1
  twistMotion=Locked
 }
 {
  bone1Name=dyn_spider_root
  bone2Name=dyn_spiderweb_03
  swing1LimitDegrees=10
  swing2LimitDegrees=45
  twistLimitDegrees=45
  twistMotion=Locked
 }
 {
  bone1Name=dyn_spider_back_l_01
  bone2Name=dyn_spider_root
  swing1LimitDegrees=10
  swing1Motion=Locked
  swing2LimitDegrees=15
  twistLimitDegrees=45
  twistMotion=Locked
 }
 {
  bone1Name=dyn_spider_back_l_02
  bone2Name=dyn_spider_back_l_01
  swing1LimitDegrees=2
  swing2LimitDegrees=10
  twistLimitDegrees=45
  twistMotion=Locked
 }
 {
  bone1Name=dyn_spider_back_r_01
  bone2Name=dyn_spider_root
  swing1LimitDegrees=10
  swing1Motion=Locked
  swing2LimitDegrees=15
  twistLimitDegrees=45
  twistMotion=Locked
 }
 {
  bone1Name=dyn_spider_back_r_02
  bone2Name=dyn_spider_back_r_01
  swing1LimitDegrees=0
  swing2LimitDegrees=2
  twistLimitDegrees=45
  twistMotion=Locked
 }
 {
  bone1Name=dyn_spider_front_l_01
  bone2Name=dyn_spider_root
  swing1LimitDegrees=2
  swing2LimitDegrees=10
  twistLimitDegrees=45
  twistMotion=Locked
 }
 {
  bone1Name=dyn_spider_front_l_02
  bone2Name=dyn_spider_front_l_01
  swing1LimitDegrees=2
  swing2LimitDegrees=10
  twistLimitDegrees=45
  twistMotion=Locked
 }
 {
  bone1Name=dyn_spider_front_r_01
  bone2Name=dyn_spider_root
  swing1LimitDegrees=2
  swing2LimitDegrees=10
  twistLimitDegrees=45
  twistMotion=Locked
 }
 {
  bone1Name=dyn_spider_front_r_02
  bone2Name=dyn_spider_front_r_01
  swing1LimitDegrees=2
  swing2LimitDegrees=10
  twistLimitDegrees=45
  twistMotion=Locked
 }
 {
  bone1Name=dyn_spider_top_middle_l_01
  bone2Name=dyn_spider_root
  swing1LimitDegrees=2
  swing1Motion=Locked
  swing2LimitDegrees=4
  twistLimitDegrees=45
  twistMotion=Locked
 }
 {
  bone1Name=dyn_spider_top_middle_l_02
  bone2Name=dyn_spider_top_middle_l_01
  swing1LimitDegrees=2
  swing2LimitDegrees=10
  twistLimitDegrees=45
  twistMotion=Locked
 }
 {
  bone1Name=dyn_spider_top_middle_l_03
  bone2Name=dyn_spider_top_middle_l_02
  swing1LimitDegrees=45
  swing1Motion=Free
  swing2LimitDegrees=45
  swing2Motion=Free
  twistLimitDegrees=45
  twistMotion=Free
 }
 {
  bone1Name=dyn_spider_top_middle_r_01
  bone2Name=dyn_spider_root
  swing1LimitDegrees=2
  swing2LimitDegrees=10
  twistLimitDegrees=45
  twistMotion=Locked
 }
 {
  bone1Name=dyn_spider_top_middle_r_02
  bone2Name=dyn_spider_top_middle_r_01
  swing1LimitDegrees=2
  swing2LimitDegrees=10
  twistLimitDegrees=45
  twistMotion=Locked
 }
 {
  bone1Name=dyn_spider_top_middle_r_03
  bone2Name=dyn_spider_top_middle_r_02
  swing1LimitDegrees=2
  swing2LimitDegrees=20
  twistLimitDegrees=45
  twistMotion=Locked
 }
]
