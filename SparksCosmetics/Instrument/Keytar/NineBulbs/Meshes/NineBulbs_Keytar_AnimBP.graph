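# Animation Blueprint for the NineBulbs keytar cosmetic that adds physics and procedural wiggling to its spider-leg bones.
# A base-bone-space RigidBody node simulates the input pose, taking an external force that flips strongly upward during skydiving or floating.
# Chained additive Modify Bone nodes then rotate the leg bones.
# Each update computes their angles as real-time cosine waves with varied speeds and slightly randomized amplitudes; Center_leg_01 is computed but unused.
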AnimBlueprint /SparksCosmetics/Instrument/Keytar/NineBulbs/Meshes/NineBulbs_Keytar_AnimBP
target_skeleton=&/NineBulbs_Keytar_Skeleton
graphs [
 AnimGraph {nodes [
   n1 = RigidBody!(ComponentPose=Local To Component(LocalPose=Input Pose()), bAlphaBoolEnabled:bool, ExternalForce=Get Gravity Override()) -> (Pose) {
    alphaInputType=Bool
    bClampLinearTranslationLimitToRefPose=true
    bOverrideWorldGravity=true
    overridePhysicsAsset=&/NineBulbs_Keytar_Physics
    simulationSpace=BaseBoneSpace
    alphaBoolBlend {blendOption=Linear}
    alphaScaleBiasClamp {bInterpResult=true, interpSpeedDecreasing=6, interpSpeedIncreasing=7}
    baseBoneRef {boneName=root}
    componentAppliedLinearAccClamp=<2000 2000 2000>
    componentLinearAccScale=<0.15 0.15 0.15>
    componentLinearVelScale=<0.15 0.15 0.15>
    overrideWorldGravity {z=-980}
    simSpaceSettings {
     dampingAlpha=0.5
     maxAngularAcceleration=15
     maxAngularVelocity=6
     maxLinearAcceleration=450
     maxLinearVelocity=800
     velocityScaleZ=0.6
     worldAlpha=1
     externalLinearDragV=<0.1 0.1 0.1>
    }
   }
   n2 = "Transform (Modify) Bone - Bone: dyn_spider_back_l_01"!(
    ComponentPose=@n1
    bAlphaBoolEnabled:bool=True
    Translation="0,0,0"
    Rotation_Roll:float=0
    Rotation_Pitch:float=Get Leg_01() -> (Leg_01:double)
    Rotation_Yaw:float=0
    Scale="1,1,1"
   ) -> (Pose) {
    alphaInputType=Bool
    rotationMode=BMM_Additive
    rotationSpace=BCS_ParentBoneSpace
    alphaBoolBlend {blendOption=Linear}
    boneToModify {boneName=dyn_spider_back_l_01}
   }
   n3 = "Transform (Modify) Bone - Bone: dyn_spider_back_l_02"!(
    ComponentPose=@n2
    bAlphaBoolEnabled:bool=True
    Translation="0,0,0"
    Rotation_Roll:float=0
    Rotation_Pitch:float=Get Leg_02() -> (Leg_02:double)
    Rotation_Yaw:float=0
    Scale="1,1,1"
   ) -> (Pose) {
    alphaInputType=Bool
    rotationMode=BMM_Additive
    rotationSpace=BCS_ParentBoneSpace
    alphaBoolBlend {blendOption=Linear}
    boneToModify {boneName=dyn_spider_back_l_02}
   }
   n4 = "Transform (Modify) Bone - Bone: dyn_spider_back_r_01"!(
    ComponentPose=@n3
    bAlphaBoolEnabled:bool=True
    Translation="0,0,0"
    Rotation_Roll:float=0
    Rotation_Pitch:float=Get Leg_01()
    Rotation_Yaw:float=0
    Scale="1,1,1"
   ) -> (Pose) {
    alphaInputType=Bool
    rotationMode=BMM_Additive
    rotationSpace=BCS_ParentBoneSpace
    alphaBoolBlend {blendOption=Linear}
    boneToModify {boneName=dyn_spider_back_r_01}
   }
   n5 = "Transform (Modify) Bone - Bone: dyn_spider_back_r_02"!(
    ComponentPose=@n4
    bAlphaBoolEnabled:bool=True
    Translation="0,0,0"
    Rotation_Roll:float=0
    Rotation_Pitch:float=Get Leg_02()
    Rotation_Yaw:float=0
    Scale="1,1,1"
   ) -> (Pose) {
    alphaInputType=Bool
    rotationMode=BMM_Additive
    rotationSpace=BCS_ParentBoneSpace
    alphaBoolBlend {blendOption=Linear}
    boneToModify {boneName=dyn_spider_back_r_02}
   }
   n6 = "Transform (Modify) Bone - Bone: dyn_spider_front_l_01"!(
    ComponentPose=@n5
    bAlphaBoolEnabled:bool=True
    Translation="0,0,0"
    Rotation_Roll:float=Get Leg_02()
    Rotation_Pitch:float=Get Center_leg_Quick_02() -> (Center_leg_Quick_02:double)
    Rotation_Yaw:float=0
    Scale="1,1,1"
   ) -> (Pose) {
    alphaInputType=Bool
    rotationMode=BMM_Additive
    rotationSpace=BCS_BoneSpace
    alphaBoolBlend {blendOption=Linear}
    boneToModify {boneName=dyn_spider_front_l_01}
   }
   n7 = "Transform (Modify) Bone - Bone: dyn_spider_front_l_02"!(
    ComponentPose=@n6
    bAlphaBoolEnabled:bool=True
    Translation="0,0,0"
    Rotation_Roll:float=0
    Rotation_Pitch:float=Get Leg_01()
    Rotation_Yaw:float=0
    Scale="1,1,1"
   ) -> (Pose) {
    alphaInputType=Bool
    rotationMode=BMM_Additive
    rotationSpace=BCS_BoneSpace
    alphaBoolBlend {blendOption=Linear}
    boneToModify {boneName=dyn_spider_front_l_02}
   }
   n8 = "Transform (Modify) Bone - Bone: dyn_spider_front_r_01"!(
    ComponentPose=@n7
    bAlphaBoolEnabled:bool=True
    Translation="0,0,0"
    Rotation_Roll:float=Get Leg_02()
    Rotation_Pitch:float=Get Center_leg_Quick_02()
    Rotation_Yaw:float=0
    Scale="1,1,1"
   ) -> (Pose) {
    alphaInputType=Bool
    rotationMode=BMM_Additive
    rotationSpace=BCS_BoneSpace
    alphaBoolBlend {blendOption=Linear}
    boneToModify {boneName=dyn_spider_front_r_01}
   }
   n9 = "Transform (Modify) Bone - Bone: dyn_spider_front_r_02"!(
    ComponentPose=@n8
    bAlphaBoolEnabled:bool=True
    Translation="0,0,0"
    Rotation_Roll:float=0
    Rotation_Pitch:float=Get Leg_01()
    Rotation_Yaw:float=0
    Scale="1,1,1"
   ) -> (Pose) {
    alphaInputType=Bool
    rotationMode=BMM_Additive
    rotationSpace=BCS_BoneSpace
    alphaBoolBlend {blendOption=Linear}
    boneToModify {boneName=dyn_spider_front_r_02}
   }
   n10 = Get Center_leg_05() -> (Center_leg_05:double)
   n11 = "Transform (Modify) Bone - Bone: dyn_center_leg_01"!(
    ComponentPose=@n9
    bAlphaBoolEnabled:bool=True
    Translation="0,0,0"
    Rotation_Roll:float=@n10
    Rotation_Pitch:float=@n10
    Rotation_Yaw:float=0
    Scale="1,1,1"
   ) -> (Pose) {
    alphaInputType=Bool
    rotationMode=BMM_Additive
    rotationSpace=BCS_BoneSpace
    alphaBoolBlend {blendOption=Linear}
    boneToModify {boneName=dyn_center_leg_01}
   }
   n12 = "Transform (Modify) Bone - Bone: dyn_center_leg_l_02"!(
    ComponentPose=@n11
    bAlphaBoolEnabled:bool=True
    Translation="0,0,0"
    Rotation_Roll:float=0
    Rotation_Pitch:float=Get Center_leg_02() -> (Center_leg_02:double)
    Rotation_Yaw:float=0
    Scale="1,1,1"
   ) -> (Pose) {
    alphaInputType=Bool
    rotationMode=BMM_Additive
    rotationSpace=BCS_BoneSpace
    alphaBoolBlend {blendOption=Linear}
    boneToModify {boneName=dyn_center_leg_l_02}
   }
   n13 = "Transform (Modify) Bone - Bone: dyn_center_leg_l_03"!(
    ComponentPose=@n12
    bAlphaBoolEnabled:bool=True
    Translation="0,0,0"
    Rotation_Roll:float=Get Center_leg_Quick() -> (Center_leg_Quick:double)
    Rotation_Pitch:float=Get Center_leg_03() -> (Center_leg_03:double)
    Rotation_Yaw:float=0
    Scale="1,1,1"
   ) -> (Pose) {
    alphaInputType=Bool
    rotationMode=BMM_Additive
    rotationSpace=BCS_BoneSpace
    alphaBoolBlend {blendOption=Linear}
    boneToModify {boneName=dyn_center_leg_l_03}
   }
   n14 = "Transform (Modify) Bone - Bone: dyn_spider_top_middle_l_01"!(
    ComponentPose=@n13
    bAlphaBoolEnabled:bool=True
    Translation="0,0,0"
    Rotation_Roll:float=Get Center_leg_04() -> (Center_leg_04:double)
    Rotation_Pitch:float=0
    Rotation_Yaw:float=0
    Scale="1,1,1"
   ) -> (Pose) {
    alphaInputType=Bool
    rotationMode=BMM_Additive
    rotationSpace=BCS_BoneSpace
    alphaBoolBlend {blendOption=Linear}
    boneToModify {boneName=dyn_spider_top_middle_l_01}
   }
   n15 = "Transform (Modify) Bone - Bone: dyn_spider_top_middle_l_02"!(
    ComponentPose=@n14
    bAlphaBoolEnabled:bool=True
    Translation="0,0,0"
    Rotation_Roll:float=0
    Rotation_Pitch:float=Get Center_leg_05()
    Rotation_Yaw:float=0
    Scale="1,1,1"
   ) -> (Pose) {
    alphaInputType=Bool
    rotationMode=BMM_Additive
    rotationSpace=BCS_BoneSpace
    alphaBoolBlend {blendOption=Linear}
    boneToModify {boneName=dyn_spider_top_middle_l_02}
   }
   n16 = "Transform (Modify) Bone - Bone: dyn_spider_top_middle_l_03"!(
    ComponentPose=@n15
    bAlphaBoolEnabled:bool=True
    Translation="0,0,0"
    Rotation_Roll:float=Get Center_leg_Quick()
    Rotation_Pitch:float=0
    Rotation_Yaw:float=0
    Scale="1,1,1"
   ) -> (Pose) {
    alphaInputType=Bool
    rotationMode=BMM_Additive
    rotationSpace=BCS_BoneSpace
    alphaBoolBlend {blendOption=Linear}
    boneToModify {boneName=dyn_spider_top_middle_l_03}
   }
   n17 = "Transform (Modify) Bone - Bone: dyn_center_leg_r_01"!(
    ComponentPose=@n16
    bAlphaBoolEnabled:bool=True
    Translation="0,0,0"
    Rotation_Roll:float=0
    Rotation_Pitch:float=Get Center_leg_04()
    Rotation_Yaw:float=0
    Scale="1,1,1"
   ) -> (Pose) {
    alphaInputType=Bool
    rotationMode=BMM_Additive
    rotationSpace=BCS_BoneSpace
    translationSpace=BCS_BoneSpace
    alphaBoolBlend {blendOption=Linear}
    boneToModify {boneName=dyn_center_leg_r_01}
   }
   n18 = Get Center_leg_06() -> (Center_leg_06:double)
   n19 = "Transform (Modify) Bone - Bone: dyn_center_leg_r_02"!(
    ComponentPose=@n17
    bAlphaBoolEnabled:bool=True
    Translation="0,0,0"
    Rotation_Roll:float=@n18
    Rotation_Pitch:float=@n18
    Rotation_Yaw:float=0
    Scale="1,1,1"
   ) -> (Pose) {
    alphaInputType=Bool
    rotationMode=BMM_Additive
    rotationSpace=BCS_BoneSpace
    translationSpace=BCS_BoneSpace
    alphaBoolBlend {blendOption=Linear}
    boneToModify {boneName=dyn_center_leg_r_02}
   }
   n20 = "Transform (Modify) Bone - Bone: dyn_center_leg_r_03"!(
    ComponentPose=@n19
    bAlphaBoolEnabled:bool=True
    Translation="0,0,0"
    Rotation_Roll:float=Get Center_leg_Quick()
    Rotation_Pitch:float=Get Center_leg_03()
    Rotation_Yaw:float=0
    Scale="1,1,1"
   ) -> (Pose) {
    alphaInputType=Bool
    rotationMode=BMM_Additive
    rotationSpace=BCS_BoneSpace
    translationSpace=BCS_BoneSpace
    alphaBoolBlend {blendOption=Linear}
    boneToModify {boneName=dyn_center_leg_r_03}
   }
   n21 = "Transform (Modify) Bone - Bone: dyn_spider_top_middle_r_01"!(
    ComponentPose=@n20
    bAlphaBoolEnabled:bool=True
    Translation="0,0,0"
    Rotation_Roll:float=0
    Rotation_Pitch:float=Get Center_leg_04()
    Rotation_Yaw:float=0
    Scale="1,1,1"
   ) -> (Pose) {
    alphaInputType=Bool
    rotationMode=BMM_Additive
    rotationSpace=BCS_BoneSpace
    translationSpace=BCS_BoneSpace
    alphaBoolBlend {blendOption=Linear}
    boneToModify {boneName=dyn_spider_top_middle_r_01}
   }
   n22 = "Transform (Modify) Bone - Bone: dyn_spider_top_middle_r_02"!(
    ComponentPose=@n21
    bAlphaBoolEnabled:bool=True
    Translation="0,0,0"
    Rotation_Roll:float=0
    Rotation_Pitch:float=Get Center_leg_05()
    Rotation_Yaw:float=0
    Scale="1,1,1"
   ) -> (Pose) {
    alphaInputType=Bool
    rotationMode=BMM_Additive
    rotationSpace=BCS_BoneSpace
    translationSpace=BCS_BoneSpace
    alphaBoolBlend {blendOption=Linear}
    boneToModify {boneName=dyn_spider_top_middle_r_02}
   }
   n23 = "Transform (Modify) Bone - Bone: dyn_spider_top_middle_r_03"!(
    ComponentPose=@n22
    bAlphaBoolEnabled:bool=True
    Translation="0,0,0"
    Rotation_Roll:float=Get Center_leg_Quick()
    Rotation_Pitch:float=0
    Rotation_Yaw:float=0
    Scale="1,1,1"
   ) -> (Pose) {
    alphaInputType=Bool
    rotationMode=BMM_Additive
    rotationSpace=BCS_BoneSpace
    translationSpace=BCS_BoneSpace
    alphaBoolBlend {blendOption=Linear}
    boneToModify {boneName=dyn_spider_top_middle_r_03}
   }
   Output Pose(Result=Component To Local(ComponentPose=@n23))
   Get Center_leg_06()
  ]}
 EventGraph {nodes [
   Event BlueprintUpdateAnimation()
   > n1 = Sequence()
   > n2 = Branch(Condition=OR Boolean(A=Get bIsSkydiveFloating(), B=Get bIsSkyDiving()))
   > Set Gravity Override(Gravity Override="0,0,15000")
   Set Gravity Override(execute=@n2.else, Gravity Override="0,0,-980")
   n3 = GetRealTimeSeconds()
   n4 = float * float(A=@n3, B=6)
   Set Leg_01(
    execute=@n1.then_1
    Leg_01:double=float * float(A="Cos (Radians)"!(A:double=@n4) -> (ReturnValue:double), B=RandomFloatInRange(Min=6, Max=6.25))
   ) -> (Output_Get:double)
   n5 = float * float(A=@n3, B=4)
   > Set Leg_02(Leg_02:double=float * float(A="Cos (Radians)"(A=@n5), B=RandomFloatInRange(Min=-4, Max=-4.25))) -> (Output_Get:double)
   n6 = float * float(A=@n3, B=2)
   > Set Center_leg_01(Center_leg_01:double=float * float(A="Cos (Radians)"(A=@n6), B=RandomFloatInRange(Min=18, Max=18.25))) -> (Output_Get:double)
   n7 = float * float(A=@n3, B=2)
   > Set Center_leg_02(Center_leg_02:double=float * float(A="Cos (Radians)"(A=@n7), B=RandomFloatInRange(Min=19.75, Max=20))) -> (Output_Get:double)
   n8 = float * float(A=@n3, B=1)
   > Set Center_leg_03(Center_leg_03:double=float * float(A="Cos (Radians)"(A=@n8), B=RandomFloatInRange(Min=-16, Max=-16.25))) -> (Output_Get:double)
   n9 = float * float(A=@n3, B=6)
   > Set Center_leg_04(Center_leg_04:double=float * float(A="Cos (Radians)"(A=@n9), B=RandomFloatInRange(Min=-16, Max=-16.25))) -> (Output_Get:double)
   n10 = float * float(A=@n3, B=6)
   > Set Center_leg_05(Center_leg_05:double=float * float(A="Cos (Radians)"(A=@n10), B=RandomFloatInRange(Min=10.75, Max=11))) -> (Output_Get:double)
   n11 = float * float(A=@n3, B=3)
   > Set Center_leg_06(Center_leg_06:double=float * float(A="Cos (Radians)"(A=@n11), B=RandomFloatInRange(Min=-10, Max=-10.25))) -> (Output_Get:double)
   n12 = float * float(A=@n3, B=20)
   > Set Center_leg_Quick(Center_leg_Quick:double=float * float(A="Cos (Radians)"(A=@n12), B=RandomFloatInRange(Min=-19.5, Max=-20))) -> (Output_Get:double)
   n13 = float * float(A=@n3, B=15)
   > Set Center_leg_Quick_02(Center_leg_Quick_02:double=float * float(A="Cos (Radians)"(A=@n13), B=RandomFloatInRange(Min=-2.5, Max=-3))) -> (Output_Get:double)
   TryGetPawnOwner()
  ]}
]
variables [
 Center_leg_01 {type=number}
 Center_leg_02 {type=number}
 Center_leg_03 {type=number}
 Center_leg_04 {type=number}
 Center_leg_05 {type=number}
 Center_leg_06 {type=number}
 Center_leg_Quick {type=number}
 Center_leg_Quick_02 {type=number}
 Gravity Override {type=Vector}
 Leg_01 {type=number}
 Leg_02 {type=number}
]
